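# Editor utility sub-widget for the Fort bulk find-and-replace tool that edits the filter choosing which assets to gather.
# Property views bound to the widget on PreConstruct expose included and excluded paths and classes, with class lists under collapsible foldouts.
# A button appends the current Content Browser path, preferring internal over virtual.
# Get and Set From Asset Filter convert between these variables and the filter struct; defaults exclude common asset types.
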EditorUtilityWidgetBlueprint /Game/Blueprints/Blutilities/ForDesign/FortBulkFindAndReplace/SupportWidgets/EUW_FortBulkFindAndReplace_AssetGathering
graphs [
 EventGraph {nodes [
   "On Clicked (AddCurrentContentBrowserPathButton)"() {is=title K2Node_ComponentBoundEvent}
   > Add(
    TargetArray=Get PathsList() -> (PathsList:Array of Strings)
    NewItem=Get Current Content Browser Path!() -> (Path:String) {is=title K2Node_CallFunction}
   ) {is=title K2Node_CallArrayFunction}
   Event PreConstruct()
   > SetObject(
    self=[
     Get PathsListView() -> (PathsListView:Fort Quest Tool Property View Object Reference)
     Get ExcludePathsListView() -> (ExcludePathsListView:Fort Quest Tool Property View Object Reference)
     Get ClassesAllowListView() -> (ClassesAllowListView:Fort Quest Tool Property View Object Reference)
     Get ClassesExcludeListView() -> (ClassesExcludeListView:Fort Quest Tool Property View Object Reference)
    ]
    NewObject=Self-Reference()
   )
  ]}
 Get Asset Filter {nodes [
   Get Asset Filter() {is=title K2Node_FunctionEntry}
   > Return Node!(
    execute
    Filter:Fort Bulk Find And Replace Asset Filter Structure=Make Fort Bulk Find and Replace Asset Filter!(
     Paths:Array of Strings=Get PathsList()
     ExcludePathsContaining:Array of Strings=Get ExcludePathsList() -> (ExcludePathsList:Array of Strings)
     OnlyIncludeClasses:Set of Soft Class Path Structures=Get OnlyIncludeClasses() -> (OnlyIncludeClasses:Set of Soft Class Path Structures)
     IgnoreClasses:Set of Soft Class Path Structures=Get ClassesExcludeList() -> (ClassesExcludeList:Set of Soft Class Path Structures)
    ) -> (FortBulkFindAndReplaceAssetFilter:Fort Bulk Find And Replace Asset Filter Structure) {is=title K2Node_MakeStruct}
   )
  ]}
 Get Current Content Browser Path {nodes [
   Get Current Content Browser Path() {is=title K2Node_FunctionEntry}
   n1 = BreakContentBrowserItemPath!(ItemPath:"Content Browser Item Path Structure (by ref)"=GetCurrentContentBrowserItemPath!() -> (ReturnValue:Content Browser Item Path Structure)) -> (VirtualPath:Name=None, InternalPath:Name=None)
   n2 = "Equal (Name)"(A=@n1.InternalPath)
   > Return Node!(execute, Path:String="To String (Name)"(InName=Select(Option 0:Name=@n1.InternalPath, Option 1:Name=@n1, Index=@n2) -> (ReturnValue:Name)))
  ]}
 Set From Asset Filter {nodes [
   n1 = Set From Asset Filter!() -> (then, Filter:Fort Bulk Find And Replace Asset Filter Structure) {is=title K2Node_FunctionEntry}
   n2 = Break Fort Bulk Find and Replace Asset Filter!(FortBulkFindAndReplaceAssetFilter:"Fort Bulk Find And Replace Asset Filter Structure (by ref)"=@n1) -> (
    Paths:Array of Strings
    ExcludePathsContaining:Array of Strings
    OnlyIncludeClasses:Set of Soft Class Path Structures
    IgnoreClasses:Set of Soft Class Path Structures
   ) {is=title K2Node_BreakStruct}
   > Set PathsList(PathsList:Array of Strings=@n2) -> (Output_Get:Array of Strings)
   > Set ExcludePathsList(ExcludePathsList:Array of Strings=@n2.ExcludePathsContaining) -> (Output_Get:Array of Strings)
   > Set OnlyIncludeClasses(OnlyIncludeClasses:Set of Soft Class Path Structures=@n2.OnlyIncludeClasses) -> (Output_Get:Set of Soft Class Path Structures)
   > Set ClassesExcludeList(ClassesExcludeList:Set of Soft Class Path Structures=@n2.IgnoreClasses) -> (Output_Get:Set of Soft Class Path Structures)
  ]}
]
variables [
 ClassesExcludeList {
  type="array<SoftClassPath>"
  default [
   &StaticMesh
   /Script/Engine.SoundWave
   /Script/Engine.DialogueWave
   /Script/Engine.Texture2D
   /Script/Engine.World
   /Script/Blutility.EditorUtilityWidgetBlueprint
   /Script/Engine.Blueprint
   /Script/UMGEditor.WidgetBlueprint
   /Script/Engine.MaterialInstanceConstant
   /Script/Niagara.NiagaraSystem
   /Script/GameplayAbilities.GameplayAbilityBlueprint
   /Script/Engine.Material
   /Script/Engine.AnimBlueprint
   /Script/CascadeRuntime.ParticleSystem
  ]
 }
 ExcludePathsList {type="array<string>", default [__ExternalActors__, L10N]}
 OnlyIncludeClasses {type="array<SoftClassPath>", default []}
 PathsList {type="array<string>", default [/BattlePassS23/]}
]
widgets {
 info {rootWidgetClass=&CanvasPanel, widgetCount=17}
 widgets [
  CanvasPanel_0 {widgetClassPath=&CanvasPanel, widget=CanvasPanel_0}
  VerticalBox_47 {
   widgetClassPath=&VerticalBox
   parent=CanvasPanel_0
   widget=VerticalBox_47
   slot {
    class=CanvasPanelSlot
    properties {bAutoSize=true, layoutData {alignment=<0 0>, anchors {maximum=<1 0>, minimum=<0 0>}, offsets=<ltrb 0 0 0 0>}}
   }
  }
  HorizontalBox_92 {widgetClassPath=&HorizontalBox, parent=VerticalBox_47, widget=HorizontalBox_92, slot {class=VerticalBoxSlot}}
  Settings {
   widgetClassPath=&VerticalBox
   parent=HorizontalBox_92
   widget=Settings
   slot {class=HorizontalBoxSlot, properties {size {sizeRule=Fill, value=1}}}
  }
  CommonTextBlock_176 {
   widgetClassPath=&CommonTextBlock
   parent=Settings
   widget=CommonTextBlock_176
   properties {
    bIsEnabled=false
    justification=Center
    text=Assets Settings
    visibility=Collapsed
    font {
     bForceMonospaced=false
     bMaterialIsStencil=false
     fontMaterial=null
     fontObject=&NotoSans
     letterSpacing=0
     monospacedWidth=1
     size=18
     skewAmount=0
     typefaceFontName=NotoSans-Regular
     outlineSettings {
      bApplyOutlineToDropShadows=false
      bMiteredCorners=false
      bSeparateFillAlpha=false
      outlineMaterial=null
      outlineSize=0
      outlineColor=<rgba 0 0 0 1>
     }
    }
    shadowOffset=<0 0>
   }
   slot {class=VerticalBoxSlot}
  }
  HorizontalBox_66 {
   widgetClassPath=&HorizontalBox
   parent=Settings
   widget=HorizontalBox_66
   slot {class=VerticalBoxSlot, properties {size {sizeRule=Fill, value=1}}}
  }
  VerticalBox_46 {
   widgetClassPath=&VerticalBox
   parent=HorizontalBox_66
   widget=VerticalBox_46
   slot {class=HorizontalBoxSlot, properties {size {sizeRule=Fill, value=1}}}
  }
  AddCurrentContentBrowserPathButton {
   widgetClassPath=&Button
   parent=VerticalBox_46
   widget=AddCurrentContentBrowserPathButton
   properties {widgetStyle {
     clickedSlateSound {resourceObject=null}
     disabled {
      bIsDynamicallyLoaded=false
      drawAs=RoundedBox
      imageType=NoImage
      mirroring=NoMirror
      resourceName=None
      resourceObject=null
      tiling=NoTile
      imageSize=<0 0>
      margin=<ltrb 0 0 0 0>
      outlineSettings {
       bUseBrushTransparency=false
       roundingType=FixedRadius
       width=1
       color {colorUseRule=UseColor_ColorTable, specifiedColor=<rgba 0 0 0 0>}
       cornerRadii=<4 4 4 4>
      }
      tintColor {colorUseRule=UseColor_Specified, specifiedColor=<rgba 0.0395 0.0395 0.0395 1>}
      uVRegion {bIsValid=false, max=<0 0>, min=<0 0>}
     }
     disabledForeground {colorUseRule=UseColor_ColorTable, specifiedColor=<rgba 0 0 0 0>}
     hovered {
      bIsDynamicallyLoaded=false
      drawAs=RoundedBox
      imageType=NoImage
      mirroring=NoMirror
      resourceName=None
      resourceObject=null
      tiling=NoTile
      imageSize=<32 32>
      margin=<ltrb 0 0 0 0>
      outlineSettings {
       bUseBrushTransparency=false
       roundingType=FixedRadius
       width=1
       color {colorUseRule=UseColor_ColorTable, specifiedColor=<rgba 0 0 0 0>}
       cornerRadii=<4 4 4 4>
      }
      tintColor {colorUseRule=UseColor_Specified, specifiedColor=<rgba 0.0953 0.0953 0.0953 1>}
      uVRegion {bIsValid=false, max=<0 0>, min=<0 0>}
     }
     hoveredForeground {colorUseRule=UseColor_ColorTable, specifiedColor=<rgba 0 0 0 0>}
     hoveredSlateSound {resourceObject=null}
     normal {
      bIsDynamicallyLoaded=false
      drawAs=RoundedBox
      imageType=NoImage
      mirroring=NoMirror
      resourceName=None
      resourceObject=null
      tiling=NoTile
      imageSize=<32 32>
      margin=<ltrb 0 0 0 0>
      outlineSettings {
       bUseBrushTransparency=false
       roundingType=FixedRadius
       width=1
       color {colorUseRule=UseColor_ColorTable, specifiedColor=<rgba 0 0 0 0>}
       cornerRadii=<4 4 4 4>
      }
      tintColor {colorUseRule=UseColor_Specified, specifiedColor=<rgba 0.0395 0.0395 0.0395 1>}
      uVRegion {bIsValid=false, max=<0 0>, min=<0 0>}
     }
     normalForeground {colorUseRule=UseColor_ColorTable, specifiedColor=<rgba 0 0 0 0>}
     normalPadding=<ltrb 12 1.5 12 1.5>
     pressed {
      bIsDynamicallyLoaded=false
      drawAs=RoundedBox
      imageType=NoImage
      mirroring=NoMirror
      resourceName=None
      resourceObject=null
      tiling=NoTile
      imageSize=<32 32>
      margin=<ltrb 0 0 0 0>
      outlineSettings {
       bUseBrushTransparency=false
       roundingType=FixedRadius
       width=1
       color {colorUseRule=UseColor_ColorTable, specifiedColor=<rgba 0 0 0 0>}
       cornerRadii=<4 4 4 4>
      }
      tintColor {colorUseRule=UseColor_Specified, specifiedColor=<rgba 0.0284 0.0284 0.0284 1>}
      uVRegion {bIsValid=false, max=<0 0>, min=<0 0>}
     }
     pressedForeground {colorUseRule=UseColor_ColorTable, specifiedColor=<rgba 0 0 0 0>}
     pressedPadding=<ltrb 12 2.5 12 0.5>
     pressedSlateSound {resourceObject=null}
    }}
   slot {class=VerticalBoxSlot, properties {horizontalAlignment=HAlign_Left}}
  }
  CommonTextBlock_53 {
   widgetClassPath=&CommonTextBlock
   parent=AddCurrentContentBrowserPathButton
   widget=CommonTextBlock_53
   properties {
    text=Add Current Content Browser Path
    font {
     bForceMonospaced=false
     bMaterialIsStencil=false
     fontMaterial=null
     fontObject=&NotoSans
     letterSpacing=0
     monospacedWidth=1
     size=8
     skewAmount=0
     typefaceFontName=NotoSans-Regular
     outlineSettings {
      bApplyOutlineToDropShadows=false
      bMiteredCorners=false
      bSeparateFillAlpha=false
      outlineMaterial=null
      outlineSize=0
      outlineColor=<rgba 0 0 0 1>
     }
    }
    shadowOffset=<0 0>
   }
   slot {class=ButtonSlot}
  }
  PathsListView {
   widgetClassPath=/Script/FortniteEditor.FortQuestToolPropertyView
   bIsVariable=true
   parent=VerticalBox_46
   widget=PathsListView
   properties {propertiesToShow [PathsList]}
   slot {class=VerticalBoxSlot, properties {size {sizeRule=Fill, value=1}}}
  }
  ExcludePathsListView {
   widgetClassPath=&FortQuestToolPropertyView
   bIsVariable=true
   parent=HorizontalBox_66
   widget=ExcludePathsListView
   properties {propertiesToShow [ExcludePathsList]}
   slot {class=HorizontalBoxSlot, properties {size {sizeRule=Fill, value=1}}}
  }
  VerticalBox_85 {
   widgetClassPath=&VerticalBox
   parent=HorizontalBox_66
   widget=VerticalBox_85
   slot {class=HorizontalBoxSlot, properties {size {sizeRule=Fill, value=1}}}
  }
  EUW_FortBulkFindAndReplace_Foldout {
   widgetClassPath=&/EUW_FortBulkFindAndReplace_Foldout
   bIsVariable=true
   parent=VerticalBox_85
   widget=EUW_FortBulkFindAndReplace_Foldout
   properties {
    button Opacity=0
    contentsWidget=&self:WidgetTree.ClassesExcludeListView
    default Open State=false
    font Size=12
    "header Text (Default)"=Exclude Classes
   }
   slot {class=VerticalBoxSlot}
  }
  ClassesExcludeListView {
   widgetClassPath=&FortQuestToolPropertyView
   bIsVariable=true
   parent=VerticalBox_85
   widget=ClassesExcludeListView
   properties {propertiesToShow [ClassesExcludeList]}
   slot {class=VerticalBoxSlot, properties {size {sizeRule=Fill, value=1}}}
  }
  VerticalBox {
   widgetClassPath=&VerticalBox
   parent=HorizontalBox_66
   widget=VerticalBox
   slot {class=HorizontalBoxSlot, properties {size {sizeRule=Fill, value=1}}}
  }
  EUW_FortBulkFindAndReplace_Foldout_1 {
   widgetClassPath=&EUW_FortBulkFindAndReplace_Foldout
   bIsVariable=true
   parent=VerticalBox
   widget=EUW_FortBulkFindAndReplace_Foldout_1
   properties {
    button Opacity=0
    contentsWidget=&self:WidgetTree.ClassesAllowListView
    default Open State=false
    font Size=12
    "header Text (Default)"=Include Classes
   }
   slot {class=VerticalBoxSlot}
  }
  ClassesAllowListView {
   widgetClassPath=&FortQuestToolPropertyView
   bIsVariable=true
   parent=VerticalBox
   widget=ClassesAllowListView
   properties {propertiesToShow [OnlyIncludeClasses]}
   slot {class=VerticalBoxSlot, properties {size {sizeRule=Fill, value=1}}}
  }
 ]
}
