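# Editor utility widget showing one asset's results in the Fort bulk find-and-replace design tool, as a foldable header over per-property rows.
# Set Result creates an inline row widget per matched property, keyed by property name in PropertyWidgets.
# Its checkbox skips the asset, the name button syncs the Content Browser to it.
# Get Result With Edits gathers enabled rows' values and reports whether anything was edited or excluded.
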EditorUtilityWidgetBlueprint /Game/Blueprints/Blutilities/ForDesign/FortBulkFindAndReplace/SupportWidgets/EUW_FortBulkFindAndReplace_ResultAsset
graphs [
 Clear Property Widgets {nodes [
   Clear Property Widgets() {is=title K2Node_FunctionEntry}
   > ClearChildren(self=Get PropertiesContainer() -> (PropertiesContainer:Vertical Box Object Reference))
   > Clear!(
    execute
    TargetMap:Map of Names to EUW Fort Bulk Find And Replace Result Asset Property Row Inline Object References=Get PropertyWidgets() -> (PropertyWidgets:Map of Names to EUW Fort Bulk Find And Replace Result Asset Property Row Inline Object References)
   ) -> (then) {is=title K2Node_CallFunction}
  ]}
 EventGraph {nodes [
   "On Clicked (AssetFoldoutButton)"() {is=title K2Node_ComponentBoundEvent}
   n1 = Get PropertiesContainer()
   > SetVisibility(self=@n1, InVisibility=Select(Option 0=Visible, Option 1=Collapsed, Index=IsVisible(self=@n1)))
   n2 = IsVisible(self=Get PropertiesContainer())
   > Set RenderTransform!(
    execute
    RenderTransform:Widget Transform Structure=Make Widget Transform!(
     Translation:Vector 2D Structure=(X=0,Y=0)
     Scale:Vector 2D Structure=(X=1,Y=1)
     Shear:Vector 2D Structure=(X=0,Y=0)
     Angle:float=Select(Option 0:float=180, Option 1:float=0, Index=@n2) -> (ReturnValue:float)
    ) -> (WidgetTransform:Widget Transform Structure) {is=title K2Node_MakeStruct}
    self:Widget Object Reference=Get AssetFoldoutButton() -> (AssetFoldoutButton:Button Object Reference)
   ) -> (then, Output_Get:Widget Transform Structure)
   "On Clicked (AssetButton)"() {is=title K2Node_ComponentBoundEvent}
   n3 = ToSoftObjectPath!(InAssetData:"Asset Data Structure (by ref)"=Break Fort Bulk Find and Replace Result for Asset!(FortBulkFindAndReplaceResultForAsset:"Fort Bulk Find And Replace Result For Asset Structure (by ref)"=Get Original Result() -> (Original Result:Fort Bulk Find And Replace Result For Asset Structure)) -> (AssetData:Asset Data Structure, Properties:Array of Fort Bulk Find And Replace Result For Property Structures) {is=title K2Node_BreakStruct}) -> (ReturnValue:Soft Object Path Structure)
   > SyncBrowserToObjects(AssetPaths=Make Array("[0]":String=BreakSoftObjectPath(InSoftObjectPath=@n3)) -> (Array:Array of Strings))
   n4 = "On Check State Changed (IsAssetEnabledCheckbox)"!() -> (OutputDelegate, then, bIsChecked:bool) {is=title K2Node_ComponentBoundEvent}
   > Set Asset Enabled!(execute, self, Enabled:bool=@n4.bIsChecked) -> (then) {is=title K2Node_CallFunction}
  ]}
 Get Result With Edits {nodes [
   Get Result With Edits() {is=title K2Node_FunctionEntry}
   > n1 = Branch(Condition=IsChecked(self=Get IsAssetEnabledCheckbox() -> (IsAssetEnabledCheckbox:Check Box Object Reference)))
   > n2 = Values!(execute, TargetMap:Map of Names to EUW Fort Bulk Find And Replace Result Asset Property Row Inline Object References=Get PropertyWidgets()) -> (then, Values:Array of EUW Fort Bulk Find And Replace Result Asset Property Row Inline Object References)
   > n3 = For Each Loop(Array:Array of EUW Fort Bulk Find And Replace Result Asset Property Row Inline Object References=@n2) -> (Array Element:EUW Fort Bulk Find And Replace Result Asset Property Row Inline Object Reference)
   n4 = Branch(execute=@n3.Completed, Condition=Get Had Any Edits() -> (Had Any Edits:bool))
   n5 = Get Original Result()
   > Return Node!(
    execute
    Asset Is Included:bool=true
    Had Edits:bool=true
    Results:Fort Bulk Find And Replace Result For Asset Structure=Make Fort Bulk Find and Replace Result for Asset!(
     AssetData:Asset Data Structure=Break Fort Bulk Find and Replace Result for Asset(FortBulkFindAndReplaceResultForAsset=@n5)
     Properties:Array of Fort Bulk Find And Replace Result For Property Structures=Get All Enabled Properties() -> (All Enabled Properties:Array of Fort Bulk Find And Replace Result For Property Structures)
    ) -> (FortBulkFindAndReplaceResultForAsset:Fort Bulk Find And Replace Result For Asset Structure) {is=title K2Node_MakeStruct}
   )
   Return Node(execute=@n4.else, Asset Is Included=true, Had Edits=false, Results=Get Original Result())
   Set Current Widget(execute=@n3.LoopBody, Current Widget:EUW Fort Bulk Find And Replace Result Asset Property Row Inline Object Reference=@n3) -> (Output_Get:EUW Fort Bulk Find And Replace Result Asset Property Row Inline Object Reference)
   > n6 = Get Edit Results!(
    execute
    self:EUW Fort Bulk Find And Replace Result Asset Property Row Inline Object Reference=Get Current Widget() -> (Current Widget:EUW Fort Bulk Find And Replace Result Asset Property Row Inline Object Reference)
   ) -> (then, Is Enabled:bool, Has Edit:bool, "Value (Edited or Original)":Fort Bulk Find And Replace Result For Property Structure) {is=title K2Node_CallFunction}
   > n7 = Branch(Condition=@n6)
   > n8 = Branch(Condition=@n6.Has Edit)
   > n9 = Set Had Any Edits(Had Any Edits:bool=true) -> (Output_Get:bool)
   Add(
    execute=[@n9.then, @n8.else]
    TargetArray:Array of Fort Bulk Find And Replace Result For Property Structures=Get All Enabled Properties()
    NewItem:"Fort Bulk Find And Replace Result For Property Structure (by ref)"=@n6."Value (Edited or Original)"
   ) {is=title K2Node_CallArrayFunction}
   Set Had Any Edits(execute=@n7.else, Had Any Edits=true)
   Return Node(execute=@n1.else, Asset Is Included=false, Had Edits=false)
  ]}
 Refresh Name {nodes [
   Refresh Name() {is=title K2Node_FunctionEntry}
   > n1 = Branch(Condition=IsChecked(self=Get IsAssetEnabledCheckbox()))
   n2 = Break Fort Bulk Find and Replace Result for Asset(FortBulkFindAndReplaceResultForAsset=Get Original Result())
   > Set Text(
    Text="To Text (Name)"!(InName:Name=Break Asset Data(AssetData=@n2).AssetName) -> (ReturnValue:Text)
    self=Get AssetNameView() -> (AssetNameView:Common Text Object Reference)
   )
   n3 = Get Original Result()
   n4 = "To String (Name)"(InName=Break Asset Data(AssetData=Break Fort Bulk Find and Replace Result for Asset(FortBulkFindAndReplaceResultForAsset=@n3)).AssetName)
   Set Text(execute=@n1.else, Text="To Text (String)"(InString=Append(A=@n4, B=" (skipping)")), self=Get AssetNameView())
  ]}
 Set Asset Enabled {nodes [
   n1 = Set Asset Enabled!() -> (then, Enabled:bool=~) {is=title K2Node_FunctionEntry}
   > SetIsChecked!(execute, self:Check Box Object Reference=Get IsAssetEnabledCheckbox(), InIsChecked:bool=@n1) -> (then)
   > SetVisibility(
    self=Get PropertiesContainer()
    InVisibility=Select(Option 0:ESlateVisibility Enum=Collapsed, Option 1:ESlateVisibility Enum=Visible, Index=Get Enabled() -> (Enabled:bool)) -> (ReturnValue:ESlateVisibility Enum)
   )
   > Refresh Name() {is=title K2Node_CallFunction}
  ]}
 Set Editable {nodes [
   n1 = Set Editable!() -> (then, Editable:bool=~) {is=title K2Node_FunctionEntry}
   > SetVisibility(self=Get IsAssetEnabledCheckbox(), InVisibility=Select(Option 0=Collapsed, Option 1=Visible, Index=@n1))
   > n2 = Values(TargetMap=Get PropertyWidgets())
   > n3 = For Each Loop(Array=@n2)
   > Set Edits Allowed!(
    execute
    self:EUW Fort Bulk Find And Replace Result Asset Property Row Inline Object Reference=@n3
    Edits Allowed:bool=Get Editable() -> (Editable:bool)
   ) -> (then) {is=title K2Node_CallFunction}
  ]}
 Set Result {nodes [
   n1 = Set Result!() -> (
    then
    Search Result:"Fort Bulk Find And Replace Result For Asset Structure (by ref)"
    Show Original Values:bool=true
    Show Replaced Values:bool=~
    Rich Text Search Terms:"Fort Bulk Find And Replace Search Terms Structure (by ref)"
    Editable:bool=~
   ) {is=title K2Node_FunctionEntry}
   > Set Original Result(Original Result:Fort Bulk Find And Replace Result For Asset Structure=@n1) -> (Output_Get:Fort Bulk Find And Replace Result For Asset Structure)
   > Refresh Name()
   > Clear Property Widgets() {is=title K2Node_CallFunction}
   > n2 = For Each Loop(Array:Array of Fort Bulk Find And Replace Result For Property Structures=Break Fort Bulk Find and Replace Result for Asset(FortBulkFindAndReplaceResultForAsset=Get Original Result()).Properties) -> (Array Element:Fort Bulk Find And Replace Result For Property Structure)
   Set Editable!(execute=@n2.Completed, self, Editable:bool=Get Editable()) -> (then) {is=title K2Node_CallFunction}
   Set Current Property(execute=@n2.LoopBody, Current Property:Fort Bulk Find And Replace Result For Property Structure=@n2) -> (Output_Get:Fort Bulk Find And Replace Result For Property Structure)
   > n3 = Create Widget!(
    execute:Exec
    WorldContextObject:Object Reference
    Class:User Widget Class Reference=&/EUW_FortBulkFindAndReplace_ResultAssetPropertyRowInline.EUW_FortBulkFindAndReplace_ResultAssetPropertyRowInline_C
    OwningPlayer:Player Controller Object Reference
   ) -> (then:Exec, ReturnValue:EUW Fort Bulk Find And Replace Result Asset Property Row Inline Object Reference)
   > Set Current Property Widget(Current Property Widget:EUW Fort Bulk Find And Replace Result Asset Property Row Inline Object Reference=@n3) -> (Output_Get:EUW Fort Bulk Find And Replace Result Asset Property Row Inline Object Reference)
   > Add!(
    execute
    TargetMap:Map of Names to EUW Fort Bulk Find And Replace Result Asset Property Row Inline Object References=Get PropertyWidgets()
    Key:"Name (by ref)"=Break Fort Bulk Find and Replace Result for Property!(FortBulkFindAndReplaceResultForProperty:"Fort Bulk Find And Replace Result For Property Structure (by ref)"=Get Current Property() -> (Current Property:Fort Bulk Find And Replace Result For Property Structure)) -> (PropertyName:Name, OriginalValue:String, ReplacedValue:String) {is=title K2Node_BreakStruct}
    Value:"EUW Fort Bulk Find And Replace Result Asset Property Row Inline Object Reference (by ref)"=Get Current Property Widget() -> (Current Property Widget:EUW Fort Bulk Find And Replace Result Asset Property Row Inline Object Reference)
   ) -> (then) {is=title K2Node_CallFunction}
   > AddChild(self=Get PropertiesContainer(), Content=Get Current Property Widget())
   > Set Result!(
    execute
    self:EUW Fort Bulk Find And Replace Result Asset Property Row Inline Object Reference=Get Current Property Widget()
    Original Property:"Fort Bulk Find And Replace Result For Property Structure (by ref)"=Get Current Property()
    RichTextSearchTerms:"Fort Bulk Find And Replace Search Terms Structure (by ref)"=Get Rich Text Search Terms() -> (Rich Text Search Terms:"Fort Bulk Find And Replace Search Terms Structure (by ref)")
    Show Original Value:bool=Get Show Original Values() -> (Show Original Values:bool)
    Show Replaced Value:bool=Get Show Replaced Values() -> (Show Replaced Values:bool)
   ) -> (then) {is=title K2Node_CallFunction}
  ]}
]
variables [
 Original Result {type=FortBulkFindAndReplaceResultForAsset, default {assetData {}, properties []}}
 PropertyWidgets {type="map<EUW_FortBulkFindAndReplace_ResultAssetPropertyRowInline_C>", default {}}
]
widgets {
 info {rootWidgetClass=&CanvasPanel, widgetCount=13}
 widgets [
  CanvasPanel_0 {widgetClassPath=&CanvasPanel, widget=CanvasPanel_0}
  VerticalBox_40 {
   widgetClassPath=&VerticalBox
   parent=CanvasPanel_0
   widget=VerticalBox_40
   slot {
    class=CanvasPanelSlot
    properties {bAutoSize=true, layoutData {alignment=<0 0>, anchors {maximum=<1 1>, minimum=<0 0>}, offsets=<ltrb 0 0 0 0>}}
   }
  }
  Header {widgetClassPath=&HorizontalBox, parent=VerticalBox_40, widget=Header, slot {class=VerticalBoxSlot}}
  SizeBox {
   widgetClassPath=&SizeBox
   parent=Header
   widget=SizeBox
   properties {bOverride_MaxAspectRatio=true, bOverride_MinAspectRatio=true, renderTransformPivot=<0 0.5>}
   slot {class=HorizontalBoxSlot, properties {horizontalAlignment=HAlign_Right, verticalAlignment=VAlign_Center}}
  }
  AssetFoldoutButton {
   widgetClassPath=&Button
   parent=SizeBox
   widget=AssetFoldoutButton
   properties {widgetStyle {
     clickedSlateSound {resourceObject=null}
     disabled {
      bIsDynamicallyLoaded=false
      drawAs=Image
      imageType=NoImage
      mirroring=NoMirror
      resourceName=None
      resourceObject=/Game/Athena/Playlists/Uranium/Art/T-Icon-Uranium_P-MeterArrowPointer
      tiling=NoTile
      imageSize=<24 24>
      margin=<ltrb 0 0 0 0>
      outlineSettings {
       bUseBrushTransparency=false
       roundingType=FixedRadius
       width=1
       color {colorUseRule=UseColor_ColorTable, specifiedColor=<rgba 0 0 0 0>}
       cornerRadii=<4 4 4 4>
      }
      tintColor {colorUseRule=UseColor_Specified, specifiedColor=<rgba 0.0395 0.0395 0.0395 1>}
      uVRegion {bIsValid=false, max=<0 0>, min=<0 0>}
     }
     disabledForeground {colorUseRule=UseColor_ColorTable, specifiedColor=<rgba 0 0 0 0>}
     hovered {
      bIsDynamicallyLoaded=false
      drawAs=Image
      imageType=NoImage
      mirroring=NoMirror
      resourceName=None
      resourceObject=&T-Icon-Uranium_P-MeterArrowPointer
      tiling=NoTile
      imageSize=<24 24>
      margin=<ltrb 0 0 0 0>
      outlineSettings {
       bUseBrushTransparency=false
       roundingType=FixedRadius
       width=1
       color {colorUseRule=UseColor_ColorTable, specifiedColor=<rgba 0 0 0 0>}
       cornerRadii=<4 4 4 4>
      }
      tintColor {colorUseRule=UseColor_Specified, specifiedColor=<rgba 0.0953 0.0953 0.0953 1>}
      uVRegion {bIsValid=false, max=<0 0>, min=<0 0>}
     }
     hoveredForeground {colorUseRule=UseColor_ColorTable, specifiedColor=<rgba 0 0 0 0>}
     hoveredSlateSound {resourceObject=null}
     normal {
      bIsDynamicallyLoaded=false
      drawAs=Image
      imageType=NoImage
      mirroring=NoMirror
      resourceName=None
      resourceObject=&T-Icon-Uranium_P-MeterArrowPointer
      tiling=NoTile
      imageSize=<24 24>
      margin=<ltrb 0 0 0 0>
      outlineSettings {
       bUseBrushTransparency=false
       roundingType=FixedRadius
       width=1
       color {colorUseRule=UseColor_ColorTable, specifiedColor=<rgba 0 0 0 0>}
       cornerRadii=<4 4 4 4>
      }
      tintColor {colorUseRule=UseColor_Specified, specifiedColor=<rgba 0.5271 0.5271 0.5271 1>}
      uVRegion {bIsValid=false, max=<0 0>, min=<0 0>}
     }
     normalForeground {colorUseRule=UseColor_ColorTable, specifiedColor=<rgba 0 0 0 0>}
     normalPadding=<ltrb 12 1.5 12 1.5>
     pressed {
      bIsDynamicallyLoaded=false
      drawAs=Image
      imageType=NoImage
      mirroring=NoMirror
      resourceName=None
      resourceObject=&T-Icon-Uranium_P-MeterArrowPointer
      tiling=NoTile
      imageSize=<24 24>
      margin=<ltrb 0 0 0 0>
      outlineSettings {
       bUseBrushTransparency=false
       roundingType=FixedRadius
       width=1
       color {colorUseRule=UseColor_ColorTable, specifiedColor=<rgba 0 0 0 0>}
       cornerRadii=<4 4 4 4>
      }
      tintColor {colorUseRule=UseColor_Specified, specifiedColor=<rgba 0.9948 0.9948 0.9948 1>}
      uVRegion {bIsValid=false, max=<0 0>, min=<0 0>}
     }
     pressedForeground {colorUseRule=UseColor_ColorTable, specifiedColor=<rgba 0 0 0 0>}
     pressedPadding=<ltrb 12 2.5 12 0.5>
     pressedSlateSound {resourceObject=null}
    }}
   slot {class=SizeBoxSlot, properties {horizontalAlignment=HAlign_Center, verticalAlignment=VAlign_Center}}
  }
  AssetButton {
   widgetClassPath=&Button
   parent=Header
   widget=AssetButton
   properties {widgetStyle {
     clickedSlateSound {resourceObject=null}
     disabled {
      bIsDynamicallyLoaded=false
      drawAs=RoundedBox
      imageType=NoImage
      mirroring=NoMirror
      resourceName=None
      resourceObject=null
      tiling=NoTile
      imageSize=<0 0>
      margin=<ltrb 0 0 0 0>
      outlineSettings {
       bUseBrushTransparency=false
       roundingType=FixedRadius
       width=1
       color {colorUseRule=UseColor_ColorTable, specifiedColor=<rgba 0 0 0 0>}
       cornerRadii=<4 4 4 4>
      }
      tintColor {colorUseRule=UseColor_Specified, specifiedColor=<rgba 0.0395 0.0395 0.0395 1>}
      uVRegion {bIsValid=false, max=<0 0>, min=<0 0>}
     }
     disabledForeground {colorUseRule=UseColor_ColorTable, specifiedColor=<rgba 0 0 0 0>}
     hovered {
      bIsDynamicallyLoaded=false
      drawAs=RoundedBox
      imageType=NoImage
      mirroring=NoMirror
      resourceName=None
      resourceObject=null
      tiling=NoTile
      imageSize=<32 32>
      margin=<ltrb 0 0 0 0>
      outlineSettings {
       bUseBrushTransparency=false
       roundingType=FixedRadius
       width=1
       color {colorUseRule=UseColor_ColorTable, specifiedColor=<rgba 0 0 0 0>}
       cornerRadii=<4 4 4 4>
      }
      tintColor {colorUseRule=UseColor_Specified, specifiedColor=<rgba 0.0953 0.0953 0.0953 1>}
      uVRegion {bIsValid=false, max=<0 0>, min=<0 0>}
     }
     hoveredForeground {colorUseRule=UseColor_ColorTable, specifiedColor=<rgba 1 1 1 1>}
     hoveredSlateSound {resourceObject=null}
     normal {
      bIsDynamicallyLoaded=false
      drawAs=RoundedBox
      imageType=NoImage
      mirroring=NoMirror
      resourceName=None
      resourceObject=null
      tiling=NoTile
      imageSize=<32 32>
      margin=<ltrb 0 0 0 0>
      outlineSettings {
       bUseBrushTransparency=false
       roundingType=FixedRadius
       width=1
       color {colorUseRule=UseColor_ColorTable, specifiedColor=<rgba 0 0 0 0>}
       cornerRadii=<4 4 4 4>
      }
      tintColor {colorUseRule=UseColor_Specified, specifiedColor=<rgba 0.0395 0.0395 0.0395 1>}
      uVRegion {bIsValid=false, max=<0 0>, min=<0 0>}
     }
     normalForeground {colorUseRule=UseColor_ColorTable, specifiedColor=<rgba 0.8 0.8 0.8 1>}
     normalPadding=<ltrb 0 0 0 0>
     pressed {
      bIsDynamicallyLoaded=false
      drawAs=RoundedBox
      imageType=NoImage
      mirroring=NoMirror
      resourceName=None
      resourceObject=null
      tiling=NoTile
      imageSize=<32 32>
      margin=<ltrb 0 0 0 0>
      outlineSettings {
       bUseBrushTransparency=false
       roundingType=FixedRadius
       width=1
       color {colorUseRule=UseColor_ColorTable, specifiedColor=<rgba 0 0 0 0>}
       cornerRadii=<4 4 4 4>
      }
      tintColor {colorUseRule=UseColor_Specified, specifiedColor=<rgba 0.0284 0.0284 0.0284 1>}
      uVRegion {bIsValid=false, max=<0 0>, min=<0 0>}
     }
     pressedForeground {colorUseRule=UseColor_ColorTable, specifiedColor=<rgba 0.5 0.5 0.5 1>}
     pressedPadding=<ltrb 0 0 0 0>
     pressedSlateSound {resourceObject=null}
    }}
   slot {class=HorizontalBoxSlot, properties {padding=<ltrb 15 0 0 0>, size {sizeRule=Fill, value=1}}}
  }
  HorizontalBox_51 {
   widgetClassPath=&HorizontalBox
   parent=AssetButton
   widget=HorizontalBox_51
   slot {class=ButtonSlot, properties {horizontalAlignment=HAlign_Fill, verticalAlignment=VAlign_Fill, padding=<ltrb 0 0 0 0>}}
  }
  AssetNameView {
   widgetClassPath=&CommonTextBlock
   bIsVariable=true
   parent=HorizontalBox_51
   widget=AssetNameView
   properties {
    text=Text Block
    font {
     bForceMonospaced=false
     bMaterialIsStencil=false
     fontMaterial=null
     fontObject=&NotoSans
     letterSpacing=0
     monospacedWidth=1
     size=10
     skewAmount=0
     typefaceFontName=NotoSans-Regular
     outlineSettings {
      bApplyOutlineToDropShadows=false
      bMiteredCorners=false
      bSeparateFillAlpha=false
      outlineMaterial=null
      outlineSize=0
      outlineColor=<rgba 0 0 0 1>
     }
    }
    shadowOffset=<0 0>
   }
   slot {class=HorizontalBoxSlot, properties {horizontalAlignment=HAlign_Center, verticalAlignment=VAlign_Center, padding=<ltrb 10 0 10 0>}}
  }
  SizeBox_1 {
   widgetClassPath=&SizeBox
   parent=HorizontalBox_51
   widget=SizeBox_1
   properties {bOverride_MaxAspectRatio=true, bOverride_MinAspectRatio=true, visibility=Visible, renderTransformPivot=<0 0.5>}
   slot {class=HorizontalBoxSlot, properties {horizontalAlignment=HAlign_Right, verticalAlignment=VAlign_Center, padding=<ltrb 0 0 10 0>}}
  }
  Image_95 {
   widgetClassPath=&Image
   bIsVariable=true
   parent=SizeBox_1
   widget=Image_95
   properties {brush {
     bIsDynamicallyLoaded=false
     drawAs=Image
     imageType=NoImage
     mirroring=NoMirror
     resourceName=None
     resourceObject=/Game/Athena/UI/InGame/Creative/PropertyEditor/Art/Icon-PropertyEditor-Search
     tiling=NoTile
     imageSize=<12 12>
     margin=<ltrb 0 0 0 0>
     outlineSettings {
      bUseBrushTransparency=false
      roundingType=HalfHeightRadius
      width=0
      color {colorUseRule=UseColor_Specified, specifiedColor=<rgba 0 0 0 0>}
      cornerRadii=<0 0 0 0>
     }
     tintColor {colorUseRule=UseColor_Specified, specifiedColor=<rgba 1 1 1 1>}
     uVRegion {bIsValid=false, max=<0 0>, min=<0 0>}
    }}
   slot {class=SizeBoxSlot}
  }
  Image_43 {
   widgetClassPath=&Image
   bIsVariable=true
   parent=Header
   widget=Image_43
   properties {
    brush {
     bIsDynamicallyLoaded=false
     drawAs=Image
     imageType=NoImage
     mirroring=NoMirror
     resourceName=None
     resourceObject=null
     tiling=NoTile
     imageSize=<32 2>
     margin=<ltrb 0 0 0 0>
     outlineSettings {
      bUseBrushTransparency=false
      roundingType=HalfHeightRadius
      width=0
      color {colorUseRule=UseColor_Specified, specifiedColor=<rgba 0 0 0 0>}
      cornerRadii=<0 0 0 0>
     }
     tintColor {colorUseRule=UseColor_Specified, specifiedColor=<rgba 1 1 1 1>}
     uVRegion {bIsValid=false, max=<0 0>, min=<0 0>}
    }
    colorAndOpacity=<rgba 1 1 1 0.1>
   }
   slot {class=HorizontalBoxSlot, properties {verticalAlignment=VAlign_Center, padding=<ltrb 20 0 20 0>, size {sizeRule=Fill, value=3}}}
  }
  IsAssetEnabledCheckbox {
   widgetClassPath=/Script/UMG.CheckBox
   bIsVariable=true
   parent=Header
   widget=IsAssetEnabledCheckbox
   properties {
    checkedState=Checked
    toolTipText=Should this asset be ignored/left out of the replace operation?
    widgetStyle {
     checkBoxType=CheckBox
     backgroundHoveredImage {
      bIsDynamicallyLoaded=false
      drawAs=RoundedBox
      imageType=NoImage
      mirroring=NoMirror
      resourceName=None
      resourceObject=null
      tiling=NoTile
      imageSize=<18 18>
      margin=<ltrb 0 0 0 0>
      outlineSettings {
       bUseBrushTransparency=false
       roundingType=FixedRadius
       width=1
       color {colorUseRule=UseColor_ColorTable, specifiedColor=<rgba 0 0 0 0>}
       cornerRadii=<3 3 3 3>
      }
      tintColor {colorUseRule=UseColor_Specified, specifiedColor=<rgba 0.0048 0.0048 0.0048 1>}
      uVRegion {bIsValid=false, max=<0 0>, min=<0 0>}
     }
     backgroundImage {
      bIsDynamicallyLoaded=false
      drawAs=RoundedBox
      imageType=NoImage
      mirroring=NoMirror
      resourceName=None
      resourceObject=null
      tiling=NoTile
      imageSize=<18 18>
      margin=<ltrb 0 0 0 0>
      outlineSettings {
       bUseBrushTransparency=false
       roundingType=FixedRadius
       width=1
       color {colorUseRule=UseColor_ColorTable, specifiedColor=<rgba 0 0 0 0>}
       cornerRadii=<3 3 3 3>
      }
      tintColor {colorUseRule=UseColor_Specified, specifiedColor=<rgba 0.0048 0.0048 0.0048 1>}
      uVRegion {bIsValid=false, max=<0 0>, min=<0 0>}
     }
     backgroundPressedImage {
      bIsDynamicallyLoaded=false
      drawAs=RoundedBox
      imageType=NoImage
      mirroring=NoMirror
      resourceName=None
      resourceObject=null
      tiling=NoTile
      imageSize=<18 18>
      margin=<ltrb 0 0 0 0>
      outlineSettings {
       bUseBrushTransparency=false
       roundingType=FixedRadius
       width=1
       color {colorUseRule=UseColor_ColorTable, specifiedColor=<rgba 0 0 0 0>}
       cornerRadii=<3 3 3 3>
      }
      tintColor {colorUseRule=UseColor_Specified, specifiedColor=<rgba 0.0048 0.0048 0.0048 1>}
      uVRegion {bIsValid=false, max=<0 0>, min=<0 0>}
     }
     borderBackgroundColor {colorUseRule=UseColor_Specified, specifiedColor=<rgba 1 1 1 1>}
     checkedForeground {colorUseRule=UseColor_Specified, specifiedColor=<rgba 1 1 1 1>}
     checkedHoveredForeground {colorUseRule=UseColor_Specified, specifiedColor=<rgba 1 1 1 1>}
     checkedHoveredImage {
      bIsDynamicallyLoaded=false
      drawAs=Image
      imageType=Vector
      mirroring=NoMirror
      resourceName=../../../Engine/Content/Slate/Starship/CoreWidgets/CheckBox/check.svg
      resourceObject=null
      tiling=NoTile
      imageSize=<16 16>
      margin=<ltrb 0 0 0 0>
      outlineSettings {
       bUseBrushTransparency=false
       roundingType=HalfHeightRadius
       width=0
       color {colorUseRule=UseColor_Specified, specifiedColor=<rgba 0 0 0 0>}
       cornerRadii=<0 0 0 0>
      }
      tintColor {colorUseRule=UseColor_Specified, specifiedColor=<rgba 0.0044 0.2384 1 1>}
      uVRegion {bIsValid=false, max=<0 0>, min=<0 0>}
     }
     checkedImage {
      bIsDynamicallyLoaded=false
      drawAs=Image
      imageType=Vector
      mirroring=NoMirror
      resourceName=../../../Engine/Content/Slate/Starship/CoreWidgets/CheckBox/check.svg
      resourceObject=null
      tiling=NoTile
      imageSize=<16 16>
      margin=<ltrb 0 0 0 0>
      outlineSettings {
       bUseBrushTransparency=false
       roundingType=HalfHeightRadius
       width=0
       color {colorUseRule=UseColor_Specified, specifiedColor=<rgba 0 0 0 0>}
       cornerRadii=<0 0 0 0>
      }
      tintColor {colorUseRule=UseColor_Specified, specifiedColor=<rgba 0 0.162 0.7454 1>}
      uVRegion {bIsValid=false, max=<0 0>, min=<0 0>}
     }
     checkedPressedForeground {colorUseRule=UseColor_Specified, specifiedColor=<rgba 1 1 1 1>}
     checkedPressedImage {
      bIsDynamicallyLoaded=false
      drawAs=Image
      imageType=Vector
      mirroring=NoMirror
      resourceName=../../../Engine/Content/Slate/Starship/CoreWidgets/CheckBox/check.svg
      resourceObject=null
      tiling=NoTile
      imageSize=<16 16>
      margin=<ltrb 0 0 0 0>
      outlineSettings {
       bUseBrushTransparency=false
       roundingType=HalfHeightRadius
       width=0
       color {colorUseRule=UseColor_Specified, specifiedColor=<rgba 0 0 0 0>}
       cornerRadii=<0 0 0 0>
      }
      tintColor {colorUseRule=UseColor_Specified, specifiedColor=<rgba 0 0.0802 0.3515 1>}
      uVRegion {bIsValid=false, max=<0 0>, min=<0 0>}
     }
     checkedSlateSound {resourceObject=null}
     foregroundColor {colorUseRule=UseColor_Specified, specifiedColor=<rgba 1 1 1 1>}
     hoveredForeground {colorUseRule=UseColor_Specified, specifiedColor=<rgba 1 1 1 1>}
     hoveredSlateSound {resourceObject=null}
     padding=<ltrb 2 0 0 0>
     pressedForeground {colorUseRule=UseColor_Specified, specifiedColor=<rgba 1 1 1 1>}
     uncheckedHoveredImage {
      bIsDynamicallyLoaded=false
      drawAs=NoDrawType
      imageType=NoImage
      mirroring=NoMirror
      resourceName=None
      resourceObject=null
      tiling=NoTile
      imageSize=<0 0>
      margin=<ltrb 0 0 0 0>
      outlineSettings {
       bUseBrushTransparency=false
       roundingType=HalfHeightRadius
       width=0
       color {colorUseRule=UseColor_Specified, specifiedColor=<rgba 0 0 0 0>}
       cornerRadii=<0 0 0 0>
      }
      tintColor {colorUseRule=UseColor_Specified, specifiedColor=<rgba 1 1 1 1>}
      uVRegion {bIsValid=false, max=<0 0>, min=<0 0>}
     }
     uncheckedImage {
      bIsDynamicallyLoaded=false
      drawAs=NoDrawType
      imageType=NoImage
      mirroring=NoMirror
      resourceName=None
      resourceObject=null
      tiling=NoTile
      imageSize=<0 0>
      margin=<ltrb 0 0 0 0>
      outlineSettings {
       bUseBrushTransparency=false
       roundingType=HalfHeightRadius
       width=0
       color {colorUseRule=UseColor_Specified, specifiedColor=<rgba 0 0 0 0>}
       cornerRadii=<0 0 0 0>
      }
      tintColor {colorUseRule=UseColor_Specified, specifiedColor=<rgba 1 1 1 1>}
      uVRegion {bIsValid=false, max=<0 0>, min=<0 0>}
     }
     uncheckedPressedImage {
      bIsDynamicallyLoaded=false
      drawAs=NoDrawType
      imageType=NoImage
      mirroring=NoMirror
      resourceName=None
      resourceObject=null
      tiling=NoTile
      imageSize=<0 0>
      margin=<ltrb 0 0 0 0>
      outlineSettings {
       bUseBrushTransparency=false
       roundingType=HalfHeightRadius
       width=0
       color {colorUseRule=UseColor_Specified, specifiedColor=<rgba 0 0 0 0>}
       cornerRadii=<0 0 0 0>
      }
      tintColor {colorUseRule=UseColor_Specified, specifiedColor=<rgba 1 1 1 1>}
      uVRegion {bIsValid=false, max=<0 0>, min=<0 0>}
     }
     uncheckedSlateSound {resourceObject=null}
     undeterminedForeground {colorUseRule=UseColor_Specified, specifiedColor=<rgba 1 1 1 1>}
     undeterminedHoveredImage {
      bIsDynamicallyLoaded=false
      drawAs=Image
      imageType=Vector
      mirroring=NoMirror
      resourceName=../../../Engine/Content/Slate/Starship/CoreWidgets/CheckBox/indeterminate.svg
      resourceObject=null
      tiling=NoTile
      imageSize=<16 16>
      margin=<ltrb 0 0 0 0>
      outlineSettings {
       bUseBrushTransparency=false
       roundingType=HalfHeightRadius
       width=0
       color {colorUseRule=UseColor_Specified, specifiedColor=<rgba 0 0 0 0>}
       cornerRadii=<0 0 0 0>
      }
      tintColor {colorUseRule=UseColor_Specified, specifiedColor=<rgba 0.0044 0.2384 1 1>}
      uVRegion {bIsValid=false, max=<0 0>, min=<0 0>}
     }
     undeterminedImage {
      bIsDynamicallyLoaded=false
      drawAs=Image
      imageType=Vector
      mirroring=NoMirror
      resourceName=../../../Engine/Content/Slate/Starship/CoreWidgets/CheckBox/indeterminate.svg
      resourceObject=null
      tiling=NoTile
      imageSize=<16 16>
      margin=<ltrb 0 0 0 0>
      outlineSettings {
       bUseBrushTransparency=false
       roundingType=HalfHeightRadius
       width=0
       color {colorUseRule=UseColor_Specified, specifiedColor=<rgba 0 0 0 0>}
       cornerRadii=<0 0 0 0>
      }
      tintColor {colorUseRule=UseColor_Specified, specifiedColor=<rgba 0 0.162 0.7454 1>}
      uVRegion {bIsValid=false, max=<0 0>, min=<0 0>}
     }
     undeterminedPressedImage {
      bIsDynamicallyLoaded=false
      drawAs=Image
      imageType=Vector
      mirroring=NoMirror
      resourceName=../../../Engine/Content/Slate/Starship/CoreWidgets/CheckBox/indeterminate.svg
      resourceObject=null
      tiling=NoTile
      imageSize=<16 16>
      margin=<ltrb 0 0 0 0>
      outlineSettings {
       bUseBrushTransparency=false
       roundingType=HalfHeightRadius
       width=0
       color {colorUseRule=UseColor_Specified, specifiedColor=<rgba 0 0 0 0>}
       cornerRadii=<0 0 0 0>
      }
      tintColor {colorUseRule=UseColor_Specified, specifiedColor=<rgba 0 0.0802 0.3515 1>}
      uVRegion {bIsValid=false, max=<0 0>, min=<0 0>}
     }
    }
   }
   slot {class=HorizontalBoxSlot}
  }
  PropertiesContainer {
   widgetClassPath=&VerticalBox
   bIsVariable=true
   parent=VerticalBox_40
   widget=PropertiesContainer
   slot {class=VerticalBoxSlot, properties {padding=<ltrb 24 0 0 0>, size {sizeRule=Fill, value=1}}}
  }
 ]
}
